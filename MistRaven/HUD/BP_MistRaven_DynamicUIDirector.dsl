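; Dynamic UI director that shows Mist Raven HUD scenes: TendrilVaccum and GrimStrut reticles, a center star, and an all-ascended notification.
; It follows the player pawn through view target changes and waits on gameplay tag add/remove tasks.
; Delta gates filter redundant scene changes, and the center star accompanies only TendrilVaccum.
; Added scenes collect in an array cleared on weapon unequip; notification removal is delayed for its out animation.
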(event EventBeginPlay
  ; Begin/EndPlay
  (bind _asathena_player_controller (Utilities|Casting|CastToAthena_PlayerController (DynamicUIDirector|GetOwningLocalPlayerController)))
  (bind _output_get (Variables|PlayerController|SetAthenaPlayerController _asathena_player_controller))
  (EventDispatchers|BindEventtoOnViewTargetChanged _output_get (EventDispatchers|CreateEvent :!event OnViewTargetChanged))
  (CallFunction|ListenforPawnEvents (Pawn|GetControlledPawn (Variables|PlayerController|GetAthenaPlayerController))))

(event EventEndPlay
  (params (EndPlayReason :type EEndPlayReason%20Enum :default "Destroyed"))
  (bind _athenaplayercontroller (Variables|PlayerController|GetAthenaPlayerController))
  ; Begin/EndPlay
  (Utilities|IsValid _athenaplayercontroller
    (:Is%20Valid
      (EventDispatchers|BindEventtoOnViewTargetChanged _athenaplayercontroller (EventDispatchers|CreateEvent :!event OnViewTargetChanged)))))

(event Custom|ListenForPawnEvents
  (params (Pawn :type Pawn%20Object%20Reference))
  (bind _asfort_player_pawn (Utilities|Casting|CastToFortPlayerPawn Pawn))
  (Variables|PlayerPawn|SetPlayerPawn _asfort_player_pawn)
  (CallFunction|ListenforonMistRavenEquip)
  (CallFunction|ListenforAscendedComplete))

(event Custom|ListenForOnMistRavenEquip
  ;[0 0 0 0.5] Equip Any Mist Raven Item
  (Ability|Tasks|WaitGameplayTagAddToActor (Variables|PlayerPawn|GetPlayerPawn) (Variables|GameplayTags|GetMistRavenWeaponEquippedTag)
    (:then
      (Ability|Tasks|WaitGameplayTagRemoveFromActor (Variables|PlayerPawn|GetPlayerPawn) (Variables|GameplayTags|GetMistRavenWeaponEquippedTag)
        (:Removed
          (CallFunction|HandleMistRavenItemUnequip))))
    (:Added
      (CallFunction|ListenforonTendrilVaccumEquip)
      (CallFunction|ListenforonGrimStrutEquip))))

(event Custom|ListenForOnTendrilVaccumEquip
  ;[0 0 0 0.5] TendrilVaccum Item Equip
  (CallFunction|OnMistRavenWeaponEquippedSceneActive (Variables|DynamicUIScenes|GetDynamicUIScene_TendrilVaccum) (Variables|GameplayTags|GetTendrilVaccumGameplayTag) (Variables|SceneActiveBooleans|GetTendrilVaccumIsActive) true true)
  (Ability|Tasks|WaitGameplayTagAddToActor (Variables|PlayerPawn|GetPlayerPawn) (Variables|GameplayTags|GetTendrilVaccumGameplayTag)
    (:then
      (Ability|Tasks|WaitGameplayTagRemoveFromActor (Variables|PlayerPawn|GetPlayerPawn) (Variables|GameplayTags|GetTendrilVaccumGameplayTag)
        (:Removed
          (CallFunction|OnMistRavenWeaponEquippedSceneActive (Variables|DynamicUIScenes|GetDynamicUIScene_TendrilVaccum) (Variables|GameplayTags|GetTendrilVaccumGameplayTag) (Variables|SceneActiveBooleans|GetTendrilVaccumIsActive)))))
    (:Added
      (CallFunction|OnMistRavenWeaponEquippedSceneActive (Variables|DynamicUIScenes|GetDynamicUIScene_TendrilVaccum) (Variables|GameplayTags|GetTendrilVaccumGameplayTag) (Variables|SceneActiveBooleans|GetTendrilVaccumIsActive) true true))))

(event Custom|ListenForOnGrimStrutEquip
  ;[0 0 0 0.5] GrimStrut Item Equip
  (CallFunction|OnMistRavenWeaponEquippedSceneActive (Variables|DynamicUIScenes|GetDynamicUIScene_GrimStrut) (Variables|GameplayTags|GetGrimStrutGameplayTag) (Variables|SceneActiveBooleans|GetGrimStrutIsActive) true)
  (Ability|Tasks|WaitGameplayTagAddToActor (Variables|PlayerPawn|GetPlayerPawn) (Variables|GameplayTags|GetGrimStrutGameplayTag)
    (:then
      (Ability|Tasks|WaitGameplayTagRemoveFromActor (Variables|PlayerPawn|GetPlayerPawn) (Variables|GameplayTags|GetGrimStrutGameplayTag)
        (:Removed
          (CallFunction|OnMistRavenWeaponEquippedSceneActive (Variables|DynamicUIScenes|GetDynamicUIScene_GrimStrut) (Variables|GameplayTags|GetGrimStrutGameplayTag) (Variables|SceneActiveBooleans|GetGrimStrutIsActive)))))
    (:Added
      (CallFunction|OnMistRavenWeaponEquippedSceneActive (Variables|DynamicUIScenes|GetDynamicUIScene_GrimStrut) (Variables|GameplayTags|GetGrimStrutGameplayTag) (Variables|SceneActiveBooleans|GetGrimStrutIsActive) true))))

(event Custom|ListenForAscendedComplete
  ;[0 0 0 0.5] MIst Raven Kit Complete
  (Ability|Tasks|WaitGameplayTagAddToActor (Variables|PlayerPawn|GetPlayerPawn) (Variables|GameplayTags|GetAllAscendedEquippedTag)
    (:then
      (Ability|Tasks|WaitGameplayTagRemoveFromActor (Variables|PlayerPawn|GetPlayerPawn) (Variables|GameplayTags|GetAllAscendedEquippedTag)
        (:Removed
          (Utilities|FlowControl|Delay 3.0) ; Ensures that the notification widget has enough time to animate out 
          (CallFunction|OnAllAscendedEquipped (Variables|SceneActiveBooleans|GetAscendedCompleteShown)))))
    (:Added
      (CallFunction|OnAllAscendedEquipped (Variables|SceneActiveBooleans|GetAscendedCompleteShown) true))))

(fn HandleMistRavenItemUnequip
  (bind _dynamicuiscenearray_mistravenreticles (Variables|DynamicUIScenes|GetDynamicUISceneArray_MistRavenReticles))
  (if (> (Utilities|Array|Length _dynamicuiscenearray_mistravenreticles) 0)
    (DynamicUIDirector|RemoveScenes _dynamicuiscenearray_mistravenreticles)
    (Variables|SceneActiveBooleans|SetTendrilVaccumIsActive false)
    (Variables|SceneActiveBooleans|SetGrimStrutIsActive false)
    (Variables|SceneActiveBooleans|SetCenterStarActive false)))

(fn OnAllAscendedEquipped
  (params (HasShownScene :type Boolean%20%28by%20ref%29) (SetSceneActive :type Boolean))
  (bind _updated_variable (Utilities|FlowControl|DeltaGateWildcard HasShownScene SetSceneActive)
    (:Success
      (bind _output_get (Variables|SceneActiveBooleans|SetAscendedCompleteShown _updated_variable))
      (if _output_get
        (DynamicUIDirector|AddScene (Variables|DynamicUIScenes|GetDynamicUIScene_AscendedNotification))
        (Utilities|Array|AddUnique (Variables|DynamicUIScenes|GetDynamicUISceneArray_MistRavenReticles) (Variables|DynamicUIScenes|GetDynamicUIScene_AscendedNotification))
        (else
          (DynamicUIDirector|RemoveScene (Variables|DynamicUIScenes|GetDynamicUIScene_AscendedNotification)))))))

(fn OnMistRavenWeaponEquippedSceneActive
  (params (DynamicUiScene :type Dynamic%20UIScene%20Object%20Reference) (ItemInEquipmentTag :type Gameplay%20Tag%20Structure) (IsSceneActive :type Boolean%20%28by%20ref%29) (SetSceneActive :type Boolean) (SetStarActive :type Boolean))
  (bind _dynamicuiscene_centerstar (Variables|DynamicUIScenes|GetDynamicUIScene_CenterStar))
  (bind _updated_variable (Utilities|FlowControl|DeltaGateWildcard IsSceneActive (Math|Boolean|ANDBoolean SetSceneActive (GameplayTags|HasMatchingGameplayTag (Variables|PlayerPawn|GetPlayerPawn) (Variables|GameplayTags|GetMistRavenWeaponEquippedTag)) (GameplayTags|HasMatchingGameplayTag (Variables|PlayerPawn|GetPlayerPawn) (Variables|GetItemInEquipmentTag))))
    (:Success
      (if _updated_variable
        (DynamicUIDirector|AddScene (Variables|GetDynamicUiScene))
        (Utilities|Array|AddUnique (Variables|DynamicUIScenes|GetDynamicUISceneArray_MistRavenReticles) (Variables|GetDynamicUiScene))
        (if (Variables|GetSetStarActive)
          (if (not (Variables|SceneActiveBooleans|GetCenterStarActive))
            (DynamicUIDirector|AddScene _dynamicuiscene_centerstar)
            (Utilities|Array|AddUnique (Variables|DynamicUIScenes|GetDynamicUISceneArray_MistRavenReticles) _dynamicuiscene_centerstar)
            (Variables|SceneActiveBooleans|SetCenterStarActive true)))
        (else
          (DynamicUIDirector|RemoveScene (Variables|GetDynamicUiScene)))))))

(fn OnPossessedPawnChanged
  (params (OldPawn :type Pawn%20Object%20Reference) (NewPawn :type Pawn%20Object%20Reference))
  (CallFunction|ListenforPawnEvents NewPawn))

(fn OnViewTargetChanged
  (params (PlayerController :type Fort%20Player%20Controller%20Object%20Reference) (OldViewTarget :type Actor%20Object%20Reference) (NewViewTarget :type Actor%20Object%20Reference))
  (Utilities|IsValid NewViewTarget
    (:Is%20Valid
      (bind _asfort_player_pawn (Utilities|Casting|CastToFortPlayerPawn NewViewTarget))
      (CallFunction|ListenforPawnEvents _asfort_player_pawn))))

(fn ConstructionScript)
